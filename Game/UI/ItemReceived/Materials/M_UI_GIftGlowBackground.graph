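# Translucent UI material for an item-received gift glow background, drawing a pulsing blue radial glow.
# An exponential radial gradient with Radius and Density parameters blends two color parameters from darker edge to lighter center.
# The same gradient forms a mask multiplied by an intensity that oscillates through a time-driven Sine scaled by TimeFactor.
# The Sine is not remapped to 0–1, so the intensity lerp also dips below Intensity-Min.
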Material /Game/UI/ItemReceived/Materials/M_UI_GIftGlowBackground
graph {
 properties {
  bUsedWithStaticMesh=false
  blendMode=BLEND_Translucent
  materialDomain=MD_UI
  pixelDepthOffsetMode=PDOM_Legacy
  usedShadingModels=Default Lit
 }
 nodes [
  n1 = MaterialFunctionCall(Radius=ScalarParameter() {defaultValue=0.5, parameterName=Radius}, Density=ScalarParameter() {defaultValue=2.25, parameterName=Density}) {materialFunction=&RadialGradientExponential}
  LinearInterpolate(
   A=VectorParameter() {parameterName=RGBA-2, defaultValue=<rgba 0 0.4361 0.85 1>}.RGB
   B=VectorParameter() {parameterName=RGBA-1, defaultValue=<rgba 0.2958 0.6571 1 1>}.RGB
   Alpha=@n1.RadialGradientExponential
  )
  n2 = Multiply(A=Time(), B=ScalarParameter() {defaultValue=0.75, parameterName=TimeFactor})
  Multiply(
   A=LinearInterpolate(Alpha=@n1.RadialGradientExponential)
   B=LinearInterpolate(
    A=ScalarParameter() {defaultValue=0.95, parameterName=Intensity-Min}
    B=ScalarParameter() {defaultValue=1.05, parameterName=Intensity-Max}
    Alpha=Sine(None=@n2)
   ) {constA=0.75}
  )
 ]
}
